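# Translucent unlit Niagara particle material for the glow on the Sunrose Pool Oracle's spline body, morphing between baked shapes.
# Dynamic parameters pick two slices of the shape position and tangent textures and blend them, building a displaced strip widened along the tangent vector by TipWidth.
# UV math derives a sphere profile from SphereCurvature for masks and normals.
# Particle color tints the glow, with a power curve shaping opacity.
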Material /BattlePassS29_Quests/Items/FX/OracleBody/M_S29_SunrosePool_OracleBody_SplineGlow
graph {
 properties {
  bUsedWithNiagaraMeshParticles=true
  bUsedWithNiagaraSprites=true
  blendMode=BLEND_Translucent
  pixelDepthOffsetMode=PDOM_Legacy
  previewMesh=&/SM_S29_OracleBody_VaryingBodySpline
  shadingModel=MSM_Unlit
  twoSided=true
  usedShadingModels=Unlit
 }
 nodes [
  n1 = Constant() {r=0.5}
  n2 = Constant() {r=1}
  n3 = ScalarParameter() {defaultValue=5, parameterName=ShapeNumber}
  n4 = Divide(
   A=Multiply(
    A=TextureCoordinate()
    B=MaterialFunctionCall(X=@n1, Y=@n2) {materialFunction=&MakeFloat2}.Result
   )
   B=MaterialFunctionCall(X=@n3, Y=Constant() {r=1}) {materialFunction=&MakeFloat2}.Result
  )
  n5 = DynamicParameter() {defaultValue=<rgba 0 1 2 0>, paramNames [Talking, ShapeIndex, NextShapeIndex, TransitionAlpha]}
  n6 = Add(A=@n5.ShapeIndex) {constB=0.1}
  n7 = Add(
   A=@n4
   B=MaterialFunctionCall(X=Divide(A=@n6, B=@n3)) {materialFunction=&MakeFloat2}.Result
  )
  n8 = Add(A=@n5.NextShapeIndex) {constB=0.1}
  n9 = Add(
   A=@n4
   B=MaterialFunctionCall(X=Divide(A=@n8, B=@n3)) {materialFunction=&MakeFloat2}.Result
  )
  n10 = LinearInterpolate(
   A=TextureSample(UVs=@n7) {samplerType=SAMPLERTYPE_LinearColor, texture=&/T_S29_OracleBody_ShapePositions}.RGB
   B=TextureSample(UVs=@n9) {samplerType=SAMPLERTYPE_LinearColor, texture=&T_S29_OracleBody_ShapePositions}.RGB
   Alpha=@n5.TransitionAlpha
  )
  n11 = ScalarParameter() {defaultValue=100, parameterName=DisplacementFactor}
  n12 = Transform(None=LinearInterpolate(
    A=TextureSample(UVs=@n7) {samplerType=SAMPLERTYPE_LinearColor, texture=&/T_S29_OracleBody_ShapeTangents}.RGB
    B=TextureSample(UVs=@n9) {samplerType=SAMPLERTYPE_LinearColor, texture=&T_S29_OracleBody_ShapeTangents}.RGB
    Alpha=@n5.TransitionAlpha
   )) {transformSourceType=TRANSFORMSOURCE_Local}
  n13 = DynamicParameter() {parameterIndex=1, defaultValue=<rgba 1 1 1 0>, paramNames [BaseWidth, TipWidth, SphereCurvature, ""]}
  n14 = TextureCoordinate()
  n15 = Multiply(
   A=LinearInterpolate(A=@n13.TipWidth, B=@n13.TipWidth, Alpha=ComponentMask(None=@n14) {g=true})
   B=ScalarParameter() {defaultValue=100, group=01-Spline, parameterName=BaseWidth, sortPriority=0}
  )
  n16 = ComponentMask(None=Add(A=Constant() {r=-0.5}, B=TextureCoordinate())) {r=true}
  Add(
   A=Transform(None=Multiply(A=@n10, B=@n11)) {transformSourceType=TRANSFORMSOURCE_Local}
   B=Multiply(A=Normalize(VectorInput=@n12), B=Multiply(A=@n15, B=@n16))
  )
  n17 = TextureCoordinate()
  n18 = VectorParameter() {parameterName=SphereDistoNoise_UV_Settings, channelNames {a=PanY, b=PanX, g=ScaleY, r=ScaleX}, defaultValue=<rgba 0.8 0.8 0 -0.75>}
  n19 = MaterialFunctionCall(X=@n18.ScaleX, Y=@n18.ScaleX) {materialFunction=&MakeFloat2}
  n20 = Time()
  n21 = MaterialFunctionCall(X=@n18.PanX, Y=@n18.PanX) {materialFunction=&MakeFloat2}
  n22 = TextureSample(UVs=Add(A=Multiply(A=@n17, B=@n19.Result), B=Multiply(A=@n20, B=@n21.Result))) {samplerType=SAMPLERTYPE_Normal, texture=/Game/Effects/Fort_Effects/Gameplay/Water/Textures/T_WaterRipples_01}
  Add(
   A=@n17
   B=Multiply(A=ComponentMask(None=@n22.RGB) {g=true, r=true}, B=ScalarParameter() {defaultValue=0.08, parameterName=SphereDistoStrength})
  )
  DynamicParameter() {defaultValue=<rgba 0 1 2 0>, paramNames [Talking, ShapeIndex, NextShapeIndex, TransitionAlpha]}
  MaterialFunctionCall(Speed=ScalarParameter() {defaultValue=0.5, parameterName=DebugSpeed}) {materialFunction=&DebugTimeSine}
  n23 = ParticleColor()
  MaterialFunctionCall(Emissive=Multiply(A=@n23.RGB, B=VectorParameter() {parameterName=GlowColor, defaultValue=<rgba 1 1 1 0.1>}.RGB)) {materialFunction=&MF_FX_EmissiveAdjustments}
  Multiply(A=@n23.A, B=Power(Base=NamedRerouteUsage(), Exp=ScalarParameter() {defaultValue=4, parameterName=GlowExp}) {constExponent=10})
  n24 = Divide(A=ConstantBiasScale(None=Saturate(None=@n17)) {bias=-0.5, scale=2}, B=ScalarParameter() {defaultValue=0.9, parameterName=SphereSize})
  n25 = DynamicParameter() {parameterIndex=1, defaultValue=<rgba 1 1 1 0>, paramNames [BaseWidth, TipWidth, SphereCurvature, ""]}
  n26 = Multiply(A=@n24, B=@n25.SphereCurvature)
  n27 = DotProduct(A=@n26, B=@n26)
  n28 = SquareRoot(None=OneMinus(None=Saturate(None=@n27)))
  n29 = Ceil(None=@n28)
  n30 = OneMinus(None=@n29)
  NamedRerouteDeclaration(None=AppendVector(A=Multiply(A=@n26, B=@n29), B=Add(A=@n30, B=@n28))) {name=SphereClampedNormal, nodeColor=<rgba 0 1 0.6953 1>}
  NamedRerouteDeclaration(None=AppendVector(A=@n26, B=@n28)) {name=SphereNormal, nodeColor=<rgba 0 1 0.9531 1>}
  NamedRerouteDeclaration(None=@n29) {name=SphereMask, nodeColor=<rgba 0 0 0 1>}
  NamedRerouteDeclaration(None=@n28) {name=SphereHeight, nodeColor=<rgba 0 0 0 1>}
  n31 = Add(A=@n28)
  NamedRerouteDeclaration(None=ConstantBiasScale(None=Divide(A=@n24, B=@n31))) {name=SphereUvs, nodeColor=<rgba 0 0 0 1>}
  ScalarParameter() {defaultValue=1, parameterName=SphereCurvature}
  ScalarParameter() {parameterName=ShapeIndex}
  ScalarParameter() {defaultValue=2, parameterName=NextShapeIndex}
 ]
}
