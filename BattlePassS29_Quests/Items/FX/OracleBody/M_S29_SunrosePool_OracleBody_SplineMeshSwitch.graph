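# Translucent unlit Niagara material for the Sunrose Pool Oracle's watery body, morphing spline shapes and glowing while talking.
# Position and tangent textures, selected and blended by dynamic-parameter shape indices and transition alpha, build a spline-like vertex offset with width.
# UV math fakes a sphere, with panning noise distorting it and deforming its height through an idle-shape mask.
# Layered color gradients, height darkening, scrolling foam and particle color form the emissive.
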Material /BattlePassS29_Quests/Items/FX/OracleBody/M_S29_SunrosePool_OracleBody_SplineMeshSwitch
graph {
 properties {
  bUsedWithNiagaraMeshParticles=true
  bUsedWithNiagaraSprites=true
  blendMode=BLEND_Translucent
  pixelDepthOffsetMode=PDOM_Legacy
  shadingModel=MSM_Unlit
  twoSided=true
  usedShadingModels=Unlit
 }
 nodes [
  n1 = Constant() {r=0.5}
  n2 = Constant() {r=1}
  n3 = ScalarParameter() {defaultValue=5, parameterName=ShapeNumber}
  n4 = Divide(
   A=Multiply(
    A=TextureCoordinate()
    B=MaterialFunctionCall(X=@n1, Y=@n2) {materialFunction=&MakeFloat2}.Result
   )
   B=MaterialFunctionCall(X=@n3, Y=Constant() {r=1}) {materialFunction=&MakeFloat2}.Result
  )
  n5 = DynamicParameter() {defaultValue=<rgba 0 0 2 0>, paramNames [Talking, ShapeIndex, NextShapeIndex, TransitionAlpha]}
  n6 = Add(A=@n5.ShapeIndex) {constB=0.1}
  n7 = Add(
   A=@n4
   B=MaterialFunctionCall(X=Divide(A=@n6, B=@n3)) {materialFunction=&MakeFloat2}.Result
  )
  n8 = Add(A=@n5.NextShapeIndex) {constB=0.1}
  n9 = Add(
   A=@n4
   B=MaterialFunctionCall(X=Divide(A=@n8, B=@n3)) {materialFunction=&MakeFloat2}.Result
  )
  n10 = LinearInterpolate(
   A=TextureSample(UVs=@n7) {samplerType=SAMPLERTYPE_LinearColor, texture=&/T_S29_OracleBody_ShapePositions}.RGB
   B=TextureSample(UVs=@n9) {samplerType=SAMPLERTYPE_LinearColor, texture=&T_S29_OracleBody_ShapePositions}.RGB
   Alpha=@n5.TransitionAlpha
  )
  n11 = ScalarParameter() {defaultValue=100, parameterName=DisplacementFactor}
  n12 = Transform(None=LinearInterpolate(
    A=TextureSample(UVs=@n7) {samplerType=SAMPLERTYPE_LinearColor, texture=&/T_S29_OracleBody_ShapeTangents}.RGB
    B=TextureSample(UVs=@n9) {samplerType=SAMPLERTYPE_LinearColor, texture=&T_S29_OracleBody_ShapeTangents}.RGB
    Alpha=@n5.TransitionAlpha
   )) {transformSourceType=TRANSFORMSOURCE_Local}
  n13 = DynamicParameter() {parameterIndex=1, defaultValue=<rgba 1 1 0.9 0.65>, paramNames [BaseWidth,TipWidth,SphereSize,DarkenHeight]}
  n14 = TextureCoordinate()
  n15 = Multiply(
   A=LinearInterpolate(A=@n13.TipWidth, B=@n13.TipWidth, Alpha=ComponentMask(None=@n14) {g=true})
   B=ScalarParameter() {defaultValue=100, group=01-Spline, parameterName=BaseWidth, sortPriority=0}
  )
  n16 = ComponentMask(None=Add(A=Constant() {r=-0.5}, B=TextureCoordinate())) {r=true}
  Add(
   A=Transform(None=Multiply(A=@n10, B=@n11)) {transformSourceType=TRANSFORMSOURCE_Local}
   B=Multiply(A=Normalize(VectorInput=@n12), B=Multiply(A=@n15, B=@n16))
  )
  MaterialFunctionCall(Speed=ScalarParameter() {defaultValue=0.5, parameterName=DebugSpeed}) {materialFunction=&DebugTimeSine}
  n17 = VectorParameter() {parameterName=DarkenColor, defaultValue=<rgba 0.1848 0.2177 0.2708 1>}
  n18 = VectorParameter() {parameterName=DarkenGlowColor, defaultValue=<rgba 0.5119 0.6486 0.8698 1>}
  n19 = DynamicParameter() {defaultValue=<rgba 0 0 2 0>, paramNames [Talking, ShapeIndex, NextShapeIndex, TransitionAlpha]}
  n20 = VectorParameter() {parameterName=MainColor, defaultValue=<rgba 0.0571 0.1389 0.25 1>}
  n21 = VectorParameter() {parameterName=GlowingMainColor, defaultValue=<rgba 0.0103 0.1658 0.4115 1>}
  n22 = VectorParameter() {parameterName=CoreColor, defaultValue=<rgba 0.034 0.0557 0.099 1>}
  n23 = VectorParameter() {parameterName=TalkingCoreColor, defaultValue=<rgba 0.3125 1.6485 3 1>}
  n24 = Power(Base=NamedRerouteUsage(), Exp=ScalarParameter() {defaultValue=2, parameterName=CoreExpo})
  n25 = LinearInterpolate(
   A=LinearInterpolate(
    A=VectorParameter() {parameterName=OuterColor, defaultValue=<rgba 0.3573 0.5957 0.7188 1>}.RGB
    B=VectorParameter() {parameterName=GlowingOuterColor, defaultValue=<rgba 0 1.4913 3 1>}.RGB
    Alpha=@n19.Talking
   )
   B=LinearInterpolate(
    A=LinearInterpolate(A=@n20.RGB, B=@n21.RGB, Alpha=@n19.Talking)
    B=LinearInterpolate(A=@n22.RGB, B=@n23.RGB, Alpha=@n19.Talking)
    Alpha=SmoothStep(Value=@n24) {constMin=0.4}
   )
   Alpha=SmoothStep(Value=NamedRerouteUsage()) {constMax=0.5}
  )
  n26 = DynamicParameter() {parameterIndex=1, defaultValue=<rgba 1 1 0.9 0.65>, paramNames [BaseWidth,TipWidth,SphereSize,DarkenHeight]}
  n27 = OneMinus(None=ComponentMask(None=NamedRerouteUsage()) {g=true})
  n28 = LinearInterpolate(
   A=Multiply(A=LinearInterpolate(A=@n17.RGB, B=@n18.RGB, Alpha=@n19.Talking), B=@n25)
   B=@n25
   Alpha=Power(Base=SmoothStep(Max=@n26.DarkenHeight, Value=@n27) {constMax=0.65})
  )
  n29 = VectorParameter() {parameterName=RippleColor, defaultValue=<rgba 0.4937 0.8202 0.9948 1>}
  n30 = NamedRerouteUsage()
  n31 = VectorParameter() {parameterName=DistoNoise_UV_Settings, channelNames {a=PanY, b=PanX, g=ScaleY, r=ScaleX}, defaultValue=<rgba 1.25 1.25 0 -0.05>}
  n32 = MaterialFunctionCall(X=@n31.ScaleX, Y=@n31.ScaleX) {materialFunction=&MakeFloat2}
  n33 = Time()
  n34 = MaterialFunctionCall(X=@n31.PanX, Y=@n31.PanX) {materialFunction=&MakeFloat2}
  n35 = TextureSample(UVs=Add(A=Multiply(A=@n30, B=@n32.Result), B=Multiply(A=@n33, B=@n34.Result))) {samplerType=SAMPLERTYPE_Normal, texture=/Game/Effects/Fort_Effects/Textures/Tile/T_Wind_N}
  n36 = Add(A=@n30, B=Multiply(A=ComponentMask(None=@n35.RGB) {g=true, r=true}, B=ScalarParameter() {defaultValue=0.02, parameterName=DistoStrength}))
  n37 = VectorParameter() {parameterName=MicroFoam_UV_Settings, channelNames {a=PanY, b=PanX, g=ScaleY, r=ScaleX}, defaultValue=<rgba 1.25 1.25 0 -0.6>}
  n38 = Multiply(
   A=@n36
   B=MaterialFunctionCall(X=@n37.ScaleX, Y=@n37.ScaleX) {materialFunction=&MakeFloat2}.Result
  )
  n39 = Multiply(
   A=Time()
   B=MaterialFunctionCall(X=@n37.PanX, Y=@n37.PanX) {materialFunction=&MakeFloat2}.Result
  )
  n40 = SmoothStep(Value=TextureSample(UVs=Add(A=@n38, B=@n39)) {texture=/Game/Effects/Fort_Effects/Textures/Tile/T_RippleTile_D}.G) {constMax=0.5, constMin=0.07}
  n41 = ScalarParameter() {defaultValue=0.5, parameterName=MicroFoamStrength}
  n42 = VectorParameter() {parameterName=MacroFoam_UV_Settings, channelNames {a=PanY, b=PanX, g=ScaleY, r=ScaleX}, defaultValue=<rgba 2 2 0.1 -1>}
  n43 = Multiply(
   A=@n36
   B=MaterialFunctionCall(X=@n42.ScaleX, Y=@n42.ScaleX) {materialFunction=&MakeFloat2}.Result
  )
  n44 = Multiply(
   A=Time()
   B=MaterialFunctionCall(X=@n42.PanX, Y=@n42.PanX) {materialFunction=&MakeFloat2}.Result
  )
  n45 = SmoothStep(Value=TextureSample(UVs=Add(A=@n43, B=@n44)) {texture=/Game/Effects/Fort_Effects/Textures/Tile/T_Noise_02}.R)
  n46 = ScalarParameter() {defaultValue=2, parameterName=MacroFoamStrength}
  n47 = OneMinus(None=NamedRerouteUsage())
  n48 = Multiply(A=Add(A=Multiply(A=@n40, B=@n41) {constB=0.5}, B=Multiply(A=@n45, B=@n46)), B=Max(B=SmoothStep(Value=@n47) {constMax=0.75}) {constA=0.2})
  n49 = NamedRerouteUsage()
  n50 = Constant2Vector() {g=0.4, r=0.6}
  n51 = Length(None=Divide(A=Subtract(A=@n49, B=@n50), B=ScalarParameter() {defaultValue=0.05, parameterName=SpecSize})) {is=class MaterialExpressionLength}
  n52 = ScalarParameter() {defaultValue=2, parameterName=SpecDensity}
  n53 = Saturate(None=OneMinus(None=Power(Base=@n51, Exp=@n52)))
  n54 = Add(A=Multiply(A=@n29.RGB, B=Saturate(None=@n48)), B=Multiply(A=Multiply(A=@n29.RGB, B=@n53)) {constB=0})
  MaterialFunctionCall(Emissive=Multiply(A=ParticleColor().RGB, B=Add(A=@n28, B=@n54))) {materialFunction=&MF_FX_EmissiveAdjustments}
  n55 = TextureCoordinate()
  n56 = VectorParameter() {parameterName=SphereDistoNoise_UV_Settings, channelNames {a=PanY, b=PanX, g=ScaleY, r=ScaleX}, defaultValue=<rgba 0.05 0.05 0 -0.03>}
  n57 = Multiply(
   A=@n55
   B=MaterialFunctionCall(X=@n56.ScaleX, Y=@n56.ScaleX) {materialFunction=&MakeFloat2}.Result
  )
  n58 = Multiply(
   A=Time()
   B=MaterialFunctionCall(X=@n56.PanX, Y=@n56.PanX) {materialFunction=&MakeFloat2}.Result
  )
  n59 = ComponentMask(None=TextureSample(UVs=Add(A=@n57, B=@n58)) {samplerType=SAMPLERTYPE_Normal, texture=/Game/Effects/Fort_Effects/Textures/WaterFoam0001_19_N}.RGB) {g=true, r=true}
  n60 = ScalarParameter() {defaultValue=0.12, parameterName=SphereDistoStrength}
  n61 = Saturate(None=Add(A=@n55, B=Multiply(A=@n59, B=@n60)))
  n62 = DynamicParameter() {parameterIndex=1, defaultValue=<rgba 1 1 0.9 0.65>, paramNames [BaseWidth,TipWidth,SphereSize,DarkenHeight]}
  n63 = Divide(A=ConstantBiasScale(None=@n61) {bias=-0.5, scale=2}, B=@n62.SphereSize)
  n64 = ScalarParameter() {defaultValue=1, parameterName=SphereCurvature}
  n65 = Multiply(A=@n63, B=@n64)
  n66 = Saturate(None=DotProduct(A=@n65, B=@n65))
  n67 = VectorParameter() {parameterName=ShapeDeformerNoise_UV_Settings, channelNames {a=PanY, b=PanX, g=ScaleY, r=ScaleX}, defaultValue=<rgba 0.4 0.3 0 -0.2>}
  n68 = Multiply(
   A=@n55
   B=MaterialFunctionCall(X=@n67.ScaleX, Y=@n67.ScaleX) {materialFunction=&MakeFloat2}.Result
  )
  n69 = Multiply(
   A=Time()
   B=MaterialFunctionCall(X=@n67.PanX, Y=@n67.PanX) {materialFunction=&MakeFloat2}.Result
  )
  n70 = Max(A=TextureSample(UVs=Add(A=@n68, B=@n69)) {texture=/Game/Effects/Fort_Effects/Textures/T_Water_Tile2}.R) {constB=0.15}
  n71 = OneMinus(None=TextureSample(UVs=@n61) {samplerSource=SSM_Clamp_WorldGroupSettings, texture=&/T_S29_OracleBody_IDLE_Shape}.R)
  n72 = ScalarParameter() {defaultValue=7, parameterName=ShapeDeformerStrength}
  n73 = DynamicParameter() {parameterIndex=2, defaultValue=<rgba 1 0 0 0>, paramNames [IsIdle, "", "", ""]}
  n74 = Subtract(A=SquareRoot(None=OneMinus(None=@n66)), B=Multiply(A=Multiply(A=@n70, B=@n71), B=Multiply(A=@n72, B=@n73.IsIdle)))
  n75 = Ceil(None=@n74)
  n76 = OneMinus(None=@n75)
  NamedRerouteDeclaration(None=AppendVector(A=Multiply(A=@n65, B=@n75), B=Add(A=@n76, B=@n74))) {name=SphereClampedNormal, nodeColor=<rgba 0 1 0.6953 1>}
  NamedRerouteDeclaration(None=AppendVector(A=@n65, B=@n74)) {name=SphereNormal, nodeColor=<rgba 0 1 0.9531 1>}
  NamedRerouteDeclaration(None=@n75) {name=SphereMask, nodeColor=<rgba 0 0 0 1>}
  NamedRerouteDeclaration(None=@n74) {name=SphereHeight, nodeColor=<rgba 0 0 0 1>}
  n77 = Divide(A=ConstantBiasScale(None=@n55) {bias=-0.5, scale=2}, B=@n62.SphereSize)
  n78 = Multiply(A=@n77, B=@n64)
  n79 = Saturate(None=DotProduct(A=@n78, B=@n78))
  n80 = Add(A=SquareRoot(None=OneMinus(None=@n79)))
  NamedRerouteDeclaration(None=ConstantBiasScale(None=Divide(A=@n77, B=@n80))) {name=SphereUVs, nodeColor=<rgba 0 0 0 1>}
  NamedRerouteUsage()
  NamedRerouteUsage()
  NamedRerouteUsage()
  n81 = Constant2Vector() {g=0.4, r=0.5}
  n82 = Length(None=Divide(A=Subtract(A=@n55, B=@n81), B=Constant2Vector() {g=0.4, r=0.3}))
  n83 = Constant() {r=2}
  Saturate(None=OneMinus(None=Power(Base=@n82, Exp=@n83)))
  n84 = Divide(A=@n63, B=Add(A=@n74))
  Saturate(None=NamedRerouteDeclaration(None=ConstantBiasScale(None=@n84)) {name=DistortedSphereUvs, nodeColor=<rgba 0 0 0 1>})
  ScalarParameter() {parameterName=ShapeIndex}
  ScalarParameter() {defaultValue=2, parameterName=NextShapeIndex}
  ScalarParameter() {defaultValue=0.4, parameterName=SpecStrength}
 ]
}
