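# Niagara idle effect for the S29 Oracle body: a glowing, shape-shifting form with splashes, falling drops, impacts and water ripples.
# System-level current and next shape indices and a timeline morph two spline-mesh layers through per-shape arrays and dynamic material parameters.
# GPU splash and drop emitters sample a shape-position texture to spawn along the next shape.
# Spawn alpha scales the effect, talk alpha drives glow, and time-of-day colors tint most emitters.
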NiagaraSystem /BattlePassS29_Quests/Items/FX/OracleBody/NS_S29_OracleBody_Idle
dependencies {
 usedDataInterfaces [
  &NiagaraDataInterfaceCamera
  &NiagaraDataInterfaceCurve
  /Script/Niagara.NiagaraDataInterfaceTexture
  &NiagaraDataInterfaceArrayFloat
  &NiagaraDataInterfaceColorCurve
  &NiagaraDataInterfaceArrayInt32
  &NiagaraDataInterfaceArrayFloat3
 ]
 usedDynamicInputs [
  &UniformAOrBFloat
  &MakeIntFromFloat
  &ModuloFloat
  &Waveform
 ]
 usedModules [
  &self:SystemSpawnScript.SystemScriptSource.SystemScriptGraph.NiagaraNodeAssignment_3.SetVariables_#1
  &self:SystemSpawnScript.SystemScriptSource.SystemScriptGraph.NiagaraNodeAssignment_4.SetVariables_#2
  &self:SystemSpawnScript.SystemScriptSource.SystemScriptGraph.NiagaraNodeAssignment_7.SetVariables_#3
  &self:SystemSpawnScript.SystemScriptSource.SystemScriptGraph.NiagaraNodeAssignment_8.SetVariables_#4
  &self:SystemSpawnScript.SystemScriptSource.SystemScriptGraph.NiagaraNodeAssignment_9.SetVariables_#5
  &self:SystemSpawnScript.SystemScriptSource.SystemScriptGraph.NiagaraNodeAssignment_12.SetVariables_#6
  &self:SystemSpawnScript.SystemScriptSource.SystemScriptGraph.NiagaraNodeAssignment_15.SetVariables_#7
  &self:SystemSpawnScript.SystemScriptSource.SystemScriptGraph.NiagaraNodeAssignment_14.SetVariables_#8
  &self:SystemSpawnScript.SystemScriptSource.SystemScriptGraph.NiagaraNodeAssignment_13.SetVariables_#9
  &SystemState
  &Timeline
  &self:SystemSpawnScript.SystemScriptSource.SystemScriptGraph.NiagaraNodeAssignment_5.SetVariables_#10
  &self:SystemSpawnScript.SystemScriptSource.SystemScriptGraph.NiagaraNodeAssignment_10.SetVariables_#11
  &self:SystemSpawnScript.SystemScriptSource.SystemScriptGraph.NiagaraNodeAssignment_2.SetVariables_#12
  &self:SystemSpawnScript.SystemScriptSource.SystemScriptGraph.NiagaraNodeAssignment_6.SetVariables_#13
  &self:SystemSpawnScript.SystemScriptSource.SystemScriptGraph.NiagaraNodeAssignment_1.SetVariables_#14
  &self:SystemSpawnScript.SystemScriptSource.SystemScriptGraph.NiagaraNodeAssignment_11.SetVariables_#15
  &self:SystemSpawnScript.SystemScriptSource.SystemScriptGraph.NiagaraNodeAssignment_0.SetVariables_#16
  &self:SwitchCurrentShape
 ]
}
emitters {
 ConstantSplashes {
  simTarget=GPUComputeSim
  emitterUpdateScript [
   &EmitterState(
    Enable Distance Culling=true
    MaxDistance=50000
    Module.bApplySpawnCountScaleByDistance=true
    Module.bUseMaxDistance=true
    Module.bUseMinDistance=false
    Scalability Mode=::NewEnumerator1
    Scale Spawn Count=true
   )
   &SpawnRate(SpawnRate=&Lerp_Float(A=0, B=20, Alpha=$System.SpawnAlpha), Use Spawn Probability=false)
  ]
  particleSpawnScript [
   &InitializeParticle(
    Color=<rgba 2.9396 4.4456 5 1>
    Lifetime=&RandomRangeFloat(Minimum=0.6, Maximum=1)
    Sprite Rotation Mode=::NewEnumerator2
    Sprite Size Mode=::NewEnumerator1
    Uniform Sprite Size Max=50
    Uniform Sprite Size Min=40
    UsePositionOffset=false
    Write Material Random=true
   )
   /Niagara/Modules/Update/Textures/SampleTexture(
    Apply Sampled RGBA to Particle Color=false
    Texture=DataInterface() {propertyValues {
      Texture=&/T_S29_OracleBody_ShapePositions
      TextureUserParameter {parameter {name=None, typeDefHandle {registeredTypeIndex=2}, varData []}}
     }}
    UV:Vector2f=&MakeVector2D(
     X=&Add_Float(
      A=&Divide_Float(
       A=&MakeFloatFromInt(Int=$System.NextShapeIndex)
       B=&MakeFloatFromInt(Int=$System.TotalShapeNumber)
      )
      B=0.1
     )
     Y=&RandomRangeFloat(Minimum=0, Maximum=1)
    )
   )
   &self:SetSplashLocation(VectorToTransform:Vector3f=&Multiply_VectorByFloat(Vector=&MakeVectorFromLinearColorRGB(Color=$Particles.SampleTexture.SampledColor), Float=100)) {module=SetSplashLocation}
   /SetVariables(Particles.Position=&AddVectorToPosition(Position=$Engine.Owner.Position, Vector=$Output.SetSplashLocation.CameraSpaceVector)) {isSetParameters=true}
   &ShapeLocation(
    Cylinder Height=0
    Cylinder Radius=&SelectFloatFromArray(Direct Array Index=$System.CurrentShapeIndex, Float Selection Array=DataInterface() {propertyValues {FloatData [35,0,0,5,5]}})
    MaskSpawnGroup=false
    Override Seed=false
    Shape Primitive=::NewEnumerator1
    Surface Only=false
    UseOwnerScale=true
   )
   &CameraOffset()
   &SubUVAnimation(End Frame=3, SubUV Animation Mode:ENiagara_SubUVLookupMode=::NewEnumerator2)
  ]
  particleUpdateScript [
   &ParticleState()
   &ScaleSpriteSize(Uniform Curve Sprite Scale=DataInterface() {propertyValues {Curve {keys [{interpMode=RCIM_Cubic, tangentMode=RCTM_User, time=0, value=0}, {interpMode=RCIM_Cubic, tangentMode=RCTM_User, time=1, value=1}]}}})
   &ScaleColor(
    Scale RGB=&MakeVectorFromLinearColorRGB(Color=&FNTimeOfDayColor(
      Night Color=<rgba 0.3734 0.4516 0.5781 1>
      Morning Color=<rgba 0.5365 0.5365 0.5365 1>
      Day Color=<rgba 1 1 1 1>
      Evening Color=<rgba 0.7917 0.791 0.5113 1>
     ))
    ScaleA=true
    ScaleRGB=true
   ) {module=ScaleColor001}
   &GravityForce(Gravity=<0 0 -150>)
   &DynamicMaterialParameters(
    Index 0 Param 1=&FloatFromCurve(
     FloatCurve=DataInterface() {propertyValues {Curve {keys [
         {arriveTangent=9.0725, interpMode=RCIM_Cubic, leaveTangent=9.0725, tangentMode=RCTM_User, time=0, value=0}
         {arriveTangent=-0.0783, interpMode=RCIM_Cubic, leaveTangent=-0.0783, tangentMode=RCTM_User, time=0.1408, value=0.9847}
         {arriveTangent=-0.0405, interpMode=RCIM_Cubic, leaveTangent=-0.0405, tangentMode=RCTM_User, time=1, value=0.5}
        ]}}}
     Scale Curve=2
    )
    Index 0 Param 2=0.1
    Index 0 Param 3=&RandomRangeFloat(Minimum=0, Maximum=0.5)
    Index 0 Param 4=1
    Module.Param0WriteEnabled=true
    Module.Param1WriteEnabled=false
    Module.Param2WriteEnabled=false
    Module.Param3WriteEnabled=false
   )
   &SolveForcesAndVelocity(Clamp Velocity=true, Limit Acceleration=false, Speed Limit=550)
   &ScaleColor(
    Scale Alpha=&FloatFromCurve(FloatCurve=DataInterface() {propertyValues {Curve {keys [
         {interpMode=RCIM_Cubic, tangentMode=RCTM_User, time=0, value=0}
         {interpMode=RCIM_Cubic, tangentMode=RCTM_User, time=0.5, value=1}
         {tangentMode=RCTM_User, time=1, value=0}
        ]}}})
    ScaleA=true
    ScaleRGB=true
   )
   &ScaleSpriteSize(
    Uniform Curve Index=&OneMinusFloat(Float=&ClampFloat(Float=&Divide_Float(
       A=&MakeFloatFromVector(
        Vector=&Subtract_Position(A=&AddVectorToPosition(Position=$Engine.Owner.Position, Vector=<0 0 -40>), B=$Particles.Position)
        Channel=::NewEnumerator2
       )
       B=50
      )))
    Uniform Curve Sprite Scale=DataInterface() {propertyValues {Curve {keys [
        {interpMode=RCIM_Cubic, tangentMode=RCTM_User, time=0, value=0}
        {arriveTangent=3.5138, interpMode=RCIM_Cubic, leaveTangent=3.5138, tangentMode=RCTM_User, time=1, value=1}
       ]}}}
   ) {module=ScaleSpriteSize002}
  ]
  rendererClasses [&NiagaraSpriteRendererProperties]
  renderers [
   {rendererClass=&NiagaraSpriteRendererProperties}
   {rendererClass=&NiagaraSpriteRendererProperties, rendererIndex=1}
  ]
 }
 DropletsImpact {
  simTarget=GPUComputeSim
  emitterUpdateScript [
   &EmitterState(
    Enable Distance Culling=true
    MaxDistance=50000
    Module.bApplySpawnCountScaleByDistance=true
    Module.bUseMaxDistance=true
    Module.bUseMinDistance=false
    Scalability Mode=::NewEnumerator1
    Scale Spawn Count=true
   )
   &SpawnRate(SpawnRate=&Lerp_Float(A=0, B=20, Alpha=$System.SpawnAlpha), Use Spawn Probability=false)
  ]
  particleSpawnScript [
   &InitializeParticle(
    Color=<rgba 10 10 10 2>
    Lifetime=&RandomRangeFloat(Minimum=0.4, Maximum=0.5)
    Sprite Rotation Mode=::NewEnumerator2
    Sprite Size Mode=::NewEnumerator1
    Uniform Sprite Size Min=8
    UsePositionOffset=false
    Write Material Random=true
   )
   &ShapeLocation(
    Cylinder Height=1
    Cylinder Radius=40
    MaskSpawnGroup=false
    Offset=$System.WaterPlaneOffset
    Offset Mode=::NewEnumerator0
    Override Seed=false
    Shape Primitive=::NewEnumerator1
    Surface Only=false
    UseOwnerScale=true
   )
   /SetVariables(Particles.InitialPos:NiagaraPosition=$Particles.Position) {isSetParameters=true}
   &CameraOffset(Camera Offset Amount=5)
   &SubUVAnimation(End Frame=3, SubUV Animation Mode:ENiagara_SubUVLookupMode=::NewEnumerator2)
  ]
  particleUpdateScript [
   &ParticleState()
   /SetVariables(Particles.SpriteFacing=<0 0 1>) {isSetParameters=true}
   &ScaleSpriteSize(Uniform Curve Sprite Scale=DataInterface() {propertyValues {Curve {keys [
        {arriveTangent=2.3005, interpMode=RCIM_Cubic, leaveTangent=2.3005, tangentMode=RCTM_User, time=0, value=0}
        {interpMode=RCIM_Cubic, tangentMode=RCTM_User, time=1, value=1}
       ]}}})
   &DynamicMaterialParameters(
    Index 0 Param 1=&FloatFromCurve(
     FloatCurve=DataInterface() {propertyValues {Curve {keys [
         {arriveTangent=9.0725, interpMode=RCIM_Cubic, leaveTangent=9.0725, tangentMode=RCTM_User, time=0, value=0}
         {arriveTangent=-0.0783, interpMode=RCIM_Cubic, leaveTangent=-0.0783, tangentMode=RCTM_User, time=0.1408, value=0.9847}
         {arriveTangent=-2.1313, interpMode=RCIM_Cubic, leaveTangent=-2.1313, tangentMode=RCTM_User, time=1, value=0}
        ]}}}
     Scale Curve=2
    )
    Index 0 Param 2=0.1
    Index 0 Param 3=&RandomRangeFloat(Minimum=0, Maximum=0.5)
    Index 0 Param 4=1
    Module.Param0WriteEnabled=true
    Module.Param1WriteEnabled=false
    Module.Param2WriteEnabled=false
    Module.Param3WriteEnabled=false
   )
   &SolveForcesAndVelocity(Clamp Velocity=true, Limit Acceleration=false, Speed Limit=550)
   &ScaleColor(
    Scale Alpha=&FloatFromCurve(FloatCurve=DataInterface() {propertyValues {Curve {keys [
         {arriveTangent=-2.0953, interpMode=RCIM_Cubic, leaveTangent=-2.0953, tangentMode=RCTM_User, time=0, value=1}
         {interpMode=RCIM_Cubic, tangentMode=RCTM_User, time=1, value=0}
        ]}}})
    ScaleA=true
    ScaleRGB=true
   )
  ]
  rendererClasses [&NiagaraSpriteRendererProperties]
  renderers [{rendererClass=&NiagaraSpriteRendererProperties}]
 }
 Dropping {
  simTarget=GPUComputeSim
  emitterUpdateScript [
   &EmitterState()
   &SpawnRate(SpawnRate=&Lerp_Float(A=0, B=10, Alpha=$System.SpawnAlpha), Use Spawn Probability=false)
  ]
  particleSpawnScript [
   &InitializeParticle(
    Color=&MultiplyLinearColor(
     Linear Color=&RandomRangeLinearColor(Minimum=<rgba 0.2779 0.6803 1 0.5>, Maximum=<rgba 1 1 1 0.5>)
     Scale Factor=&MakeVector4_FromFloat(Float=&RandomRangeFloat(Minimum=0.8, Maximum=1.2))
    )
    Lifetime=&RandomRangeFloat(Minimum=1, Maximum=0.8)
    Sprite Size Mode=::NewEnumerator3
    Uniform Sprite Size=&Multiply_Float(A=&RandomRangeFloat(Minimum=3, Maximum=5), B=0.7)
    UsePositionOffset=false
    Write Material Random=true
   )
   &FN_SetSpawnGroup(Number of Spawn Groups=5)
   &SampleTexture(
    Apply Sampled RGBA to Particle Color=false
    Texture=DataInterface() {propertyValues {
      Texture=&T_S29_OracleBody_ShapePositions
      TextureUserParameter {parameter {name=None, typeDefHandle {registeredTypeIndex=2}, varData []}}
     }}
    UV:Vector2f=&MakeVector2D(
     X=&Add_Float(
      A=&Divide_Float(
       A=&MakeFloatFromInt(Int=$System.NextShapeIndex)
       B=&MakeFloatFromInt(Int=$System.TotalShapeNumber)
      )
      B=0.1
     )
     Y=&RandomRangeFloat(Minimum=0, Maximum=1)
    )
   )
   &self:SetSplashLocation(VectorToTransform:Vector3f=&Multiply_VectorByFloat(Vector=&MakeVectorFromLinearColorRGB(Color=$Particles.SampleTexture.SampledColor), Float=100)) {module=SetSplashLocation}
   /SetVariables(Particles.Position=&AddVectorToPosition(Position=$Engine.Owner.Position, Vector=$Output.SetSplashLocation.CameraSpaceVector)) {isSetParameters=true}
   &ShapeLocation(
    Cylinder Height=0
    Cylinder Radius=&SelectFloatFromArray(Direct Array Index=$System.CurrentShapeIndex, Float Selection Array=DataInterface() {propertyValues {FloatData [35,1,1,10,10]}})
    MaskSpawnGroup=false
    Override Seed=false
    Shape Primitive=::NewEnumerator1
    Surface Only=false
    UseOwnerScale=true
   )
  ]
  particleUpdateScript [
   &ParticleState()
   &ScaleSpriteSize(Uniform Curve Sprite Scale=DataInterface() {propertyValues {Curve {keys [{arriveTangent=-2.1875, interpMode=RCIM_Cubic, time=0, value=1}, <at 1 2 RCIM_Cubic>]}}}) {module=ScaleSpriteSize001}
   &ScaleColor(
    Scale RGB=&MakeVectorFromLinearColorRGB(Color=&FNTimeOfDayColor(
      Night Color=<rgba 0.3734 0.4516 0.5781 1>
      Morning Color=<rgba 0.5365 0.5365 0.5365 1>
      Day Color=<rgba 1 1 1 1>
      Evening Color=<rgba 0.7917 0.791 0.5113 1>
     ))
    ScaleA=true
    ScaleRGB=true
   )
   &GravityForce(Gravity=&Multiply_VectorByFloat(
     Vector=<0 0 -980>
     Float=&FloatFromCurve(
      FloatCurve=DataInterface() {propertyValues {Curve {keys [{tangentMode=RCTM_User, time=0, value=0}, {tangentMode=RCTM_User, time=1, value=1}]}}}
      Scale Curve=2
     )
    ))
   &Drag(Drag=2, Use Linear Drag=true, Use Rotational Drag=true)
   &SolveForcesAndVelocity(Clamp Velocity=false, Limit Acceleration=false)
   &ScaleSpriteSizeBySpeed(Max Scale Factor=<0.5 3>, Velocity Threshold=600)
   &ScaleSpriteSize(
    Uniform Curve Scale=0.75
    Uniform Curve Sprite Scale=DataInterface() {propertyValues {Curve {keys [
        <at 0 0 RCIM_Cubic>
        {arriveTangent=2.5, interpMode=RCIM_Cubic, leaveTangent=2.5, time=0.1, value=1.5}
        {arriveTangent=-1.6667, interpMode=RCIM_Cubic, leaveTangent=-1.6667, time=0.6, value=1.5}
        {interpMode=RCIM_Cubic, leaveTangent=-8.0089, time=1, value=0}
       ]}}}
   )
   &DynamicMaterialParameters(
    Index 0 Param 1=&FloatFromCurve(FloatCurve=DataInterface() {propertyValues {Curve {keys [
         {arriveTangent=4.1785, interpMode=RCIM_Cubic, leaveTangent=4.1785, tangentMode=RCTM_User, time=0, value=0}
         {arriveTangent=0.0224, interpMode=RCIM_Cubic, leaveTangent=0.0224, tangentMode=RCTM_User, time=0.5, value=1}
        ]}}})
    Index 0 Param 2=0.75
    Index 0 Param 3=0.25
    Index 0 Param 4=1.5
    Index 1 Param 1=&RemapRange(
     Input Value To Remap=&VectorLength(Vector=$Particles.Velocity)
     Input Min=0
     Input Max=500
     Output Min=0.25
     Output Max=4
    )
    Module.Index0Param0Write=true
    Module.Index0Param1Write=true
    Module.Index0Param2Write=true
    Module.Index0Param3Write=true
    Module.Index1Param0Write=true
    Module.Index1Param1Write=true
    Module.Index1Param2Write=true
    Module.Index1Param3Write=true
    Write Parameter Index 1=true
   )
   &CameraOffset(Camera Offset Amount=30)
   &ScaleSpriteSize(
    Uniform Curve Index=&OneMinusFloat(Float=&ClampFloat(Float=&Divide_Float(
       A=&MakeFloatFromVector(
        Vector=&Subtract_Position(A=&AddVectorToPosition(Position=$Engine.Owner.Position, Vector=<0 0 -50>), B=$Particles.Position)
        Channel=::NewEnumerator2
       )
       B=50
      )))
    Uniform Curve Sprite Scale=DataInterface() {propertyValues {Curve {keys [
        {interpMode=RCIM_Cubic, tangentMode=RCTM_User, time=0, value=0}
        {arriveTangent=3.5138, interpMode=RCIM_Cubic, leaveTangent=3.5138, tangentMode=RCTM_User, time=1, value=1}
       ]}}}
   ) {module=ScaleSpriteSize002}
  ]
  rendererClasses [&NiagaraSpriteRendererProperties]
  renderers [{rendererClass=&NiagaraSpriteRendererProperties}]
 }
 Glow {
  emitterUpdateScript [
   &EmitterState(Life Cycle Mode=::NewEnumerator1, Loop Behavior=::NewEnumerator1, Loop Duration=1, UseLoopDelay=false)
   &SpawnBurst_Instantaneous(Use Spawn Probability=false, UseLoopCountLimit=false)
   &SpawnRate(SpawnRate=0.75, Use Spawn Probability=false) {enabled=false}
  ]
  particleSpawnScript [
   &InitializeParticle(
    Color=<rgba 0.1082 0.663 1 1>
    Lifetime=1.5
    Sprite Size Mode=::NewEnumerator3
    Uniform Sprite Size=150
    UsePositionOffset=true
    Write Material Random=true
   )
   &CameraOffset(Camera Offset Amount=4)
  ]
  particleUpdateScript [
   &ParticleState(Kill Particles When Lifetime Has Elapsed=false)
   /SetVariables(Particles.Position=&Multiply_VectorByFloat(Vector=&ReturnCameraVector(Camera Vectors=::NewEnumerator2, Coordinate Space=::Local), Float=35)) {isSetParameters=true}
   &ScaleColor(
    Scale Alpha=&Multiply_Float(
     A=&Lerp_Float(A=0.5, B=1, Alpha=&Max_Float(A:NiagaraFloat=$System.DebugTalkAlpha, B:NiagaraFloat=0))
     B=$System.Timeline002.NormalizedTime
    )
    Scale RGB=&MakeVectorFromLinearColorRGB(Color=&FNTimeOfDayColor(
      Night Color=<rgba 0.3734 0.4516 0.5781 1>
      Morning Color=<rgba 0.5365 0.5365 0.5365 1>
      Day Color=<rgba 1 1 1 1>
      Evening Color=<rgba 0.7917 0.791 0.5113 1>
     ))
    ScaleA=true
    ScaleRGB=true
   )
   &ScaleSpriteSize(
    Scale Sprite Size Mode=::NewEnumerator0
    Uniform Scale Factor=&Multiply_Float(
     A=&Lerp_Float(A=1, B=1, Alpha=&Max_Float(A:NiagaraFloat=$System.DebugTalkAlpha, B:NiagaraFloat=0))
     B=&Multiply_Float(
      A=$User.SystemScale
      B=&FloatFromCurve(
       FloatCurve=DataInterface() {propertyValues {Curve {keys [<at 0 0 RCIM_Cubic>, {arriveTangent=2.7075, interpMode=RCIM_Cubic, leaveTangent=2.7075, tangentMode=RCTM_User, time=1, value=1}]}}}
       CurveIndex=$System.SpawnAlpha
      )
     )
    )
   )
   &DynamicMaterialParameters(
    Index 0 Param 2=1.5
    Index 0 Param 3=1
    Index 0 Param 4=1
    Module.Index0Param0Write=true
    Module.Index0Param1Write=true
    Module.Index0Param2Write=true
    Module.Index0Param3Write=true
   )
  ]
  rendererClasses [&NiagaraSpriteRendererProperties]
  renderers [{rendererClass=&NiagaraSpriteRendererProperties}]
 }
 NE_StarGlow {
  emitterUpdateScript [
   &EmitterState()
   &SpawnRate(
    SpawnRate=&RandomRangeFloat(Minimum=-2, Maximum=5, Evaluation Type=::NewEnumerator1)
    Use Spawn Probability=false
   )
  ]
  particleSpawnScript [
   &InitializeParticle(
    Color=<rgba 0 1.1774 2 1>
    Lifetime Min=1.5
    Lifetime Mode=::NewEnumerator1
    Mass Mode=::NewEnumerator3
    Sprite Rotation Mode=::NewEnumerator2
    Sprite Size Mode=::NewEnumerator1
    Uniform Sprite Size Max=400
    Uniform Sprite Size Min=350
    UsePositionOffset=true
    Write Material Random=true
   )
   &ShapeLocation(MaskSpawnGroup=false, Non Uniform Scale=<0.5 1 1>, Override Seed=false, Sphere Radius=0, UseOwnerScale=true)
   &CameraOffset(Camera Offset Amount=4)
  ]
  particleUpdateScript [
   &ParticleState()
   /SetVariables(Particles.Position=&Multiply_VectorByFloat(Vector=&ReturnCameraVector(Camera Vectors=::NewEnumerator2, Coordinate Space=::Local), Float=35)) {isSetParameters=true}
   &SpriteRotationRate(Rotation Rate=&FloatFromCurve(
     FloatCurve=DataInterface() {propertyValues {Curve {keys [
         {arriveTangent=1.4306, interpMode=RCIM_Cubic, leaveTangent=1.4306, tangentMode=RCTM_User, time=0.2, value=0}
         {arriveTangent=1.1179, interpMode=RCIM_Cubic, leaveTangent=1.1179, tangentMode=RCTM_User, time=0.4, value=0.6228}
         {arriveTangent=0.0224, interpMode=RCIM_Cubic, leaveTangent=0.0224, tangentMode=RCTM_User, time=1, value=1}
        ]}}}
     Scale Curve=10
    ))
   &ScaleSpriteSize(
    Scale Sprite Size Mode=::NewEnumerator0
    Uniform Scale Factor=&Multiply_Float(
     A=&Lerp_Float(A=0.9, B=1.1, Alpha=&Max_Float(A:NiagaraFloat=$System.DebugTalkAlpha, B:NiagaraFloat=0))
     B=&Multiply_Float(
      A=$User.SystemScale
      B=&FloatFromCurve(FloatCurve=DataInterface() {propertyValues {Curve {keys [<at 0.15 0>,<at 1 0.9>]}}}, CurveIndex=$System.SpawnAlpha)
     )
    )
   )
   &ScaleColor(
    Scale Alpha=&Multiply_Float(
     A=&Multiply_Float(
      A=&Lerp_Float(A=0.9, B=1, Alpha=$System.GlowIntensity)
      B=&FloatFromCurve(FloatCurve=DataInterface() {propertyValues {Curve {keys [
           {interpMode=RCIM_Cubic, tangentMode=RCTM_User, time=0, value=0}
           {interpMode=RCIM_Cubic, tangentMode=RCTM_User, time=0.8, value=1}
           {interpMode=RCIM_Cubic, tangentMode=RCTM_User, time=1, value=0}
          ]}}})
     )
     B=$System.Timeline002.NormalizedTime
    )
    Scale RGB=&MakeVectorFromLinearColorRGB(Color=&FNTimeOfDayColor(
      Night Color=<rgba 0.3734 0.4516 0.5781 1>
      Morning Color=<rgba 0.5365 0.5365 0.5365 1>
      Day Color=<rgba 1 1 1 1>
      Evening Color=<rgba 0.7917 0.791 0.5113 1>
     ))
    ScaleA=true
    ScaleRGB=true
   )
  ]
  rendererClasses [&NiagaraSpriteRendererProperties]
  renderers [{rendererClass=&NiagaraSpriteRendererProperties}]
 }
 NE_WaterRipples {
  emitterUpdateScript [
   &EmitterState()
   &SpawnRate(SpawnRate=&Lerp_Float(A=0, B=10, Alpha=$System.SpawnAlpha), Use Spawn Probability=false)
  ]
  particleSpawnScript [
   &InitializeParticle(
    Color=&RandomRangeLinearColor(Minimum=<rgba 0.0208 0.1384 0.2 1>, Maximum=<rgba 0.0198 0.1696 0.2 1>)
    Lifetime Max=0.8
    Lifetime Min=0.5
    Lifetime Mode=::NewEnumerator1
    Mesh Scale=<0.15 0.2 3>
    Mesh Scale Mode=::NewEnumerator4
    Position Offset=$System.WaterPlaneOffset
    UsePositionOffset=true
    Write Material Random=true
   )
   &ShapeLocation(
    Cylinder Distribution=::NewEnumerator2
    Cylinder Height=0
    Cylinder Radius=&RandomRangeFloat(Minimum=1, Maximum=25)
    MaskSpawnGroup=false
    Shape Primitive=::NewEnumerator1
    U Position=&Multiply_Float(A=$Emitter.Age, B=4.256)
    UseOwnerScale=true
    V Position=0
   )
   &InitialMeshOrientation(
    Mesh Axis To Orient=<0 1 0>
    Mesh Orientation Mode=::NewEnumerator2
    Orientation Coordinate Space=::Local
    Orientation Vector=$Particles.ShapeLocation.ShapeTangent
    Use Rotation Vector=true
   )
  ]
  particleUpdateScript [
   &ParticleState()
   &SolveForcesAndVelocity(Clamp Velocity=false, Limit Acceleration=false)
   /SetVariables(Particles.ScaleOut:NiagaraFloat=&FloatFromCurve(
     FloatCurve=DataInterface() {propertyValues {Curve {keys [
         {arriveTangent=4.1785, interpMode=RCIM_Cubic, leaveTangent=4.1785, tangentMode=RCTM_User, time=0, value=0}
         {arriveTangent=1.1179, interpMode=RCIM_Cubic, leaveTangent=1.1179, tangentMode=RCTM_User, time=0.2042, value=0.6228}
         {arriveTangent=0.0224, interpMode=RCIM_Cubic, leaveTangent=0.0224, tangentMode=RCTM_User, time=1, value=1}
        ]}}}
     Scale Curve=3.0602
    )) {isSetParameters=true}
   &ScaleMeshSize(Scale Factor=&MakeVector(
     X=$Particles.ScaleOut
     Y=$Particles.ScaleOut
     Z=&FloatFromCurve(FloatCurve=DataInterface() {propertyValues {Curve {keys [
          {arriveTangent=7.6074, interpMode=RCIM_Cubic, leaveTangent=7.6074, tangentMode=RCTM_User, time=-0.0046, value=0}
          {arriveTangent=-1.2717, interpMode=RCIM_Cubic, leaveTangent=-1.2717, tangentMode=RCTM_User, time=0.2187, value=0.6876}
          <at 1.0011 0.002 RCIM_Cubic>
         ]}}})
    ))
   &DynamicMaterialParameters(
    Index 0 Param 1=1.5
    Index 0 Param 2=1
    Index 0 Param 3=0.125
    Index 0 Param 4=0.3
    Module.Param0WriteEnabled=true
    Module.Param1WriteEnabled=true
    Module.Param2WriteEnabled=true
    Module.Param3WriteEnabled=true
   )
   &ScaleColor(
    Scale Alpha=&FloatFromCurve(FloatCurve=DataInterface() {propertyValues {Curve {keys [
         {interpMode=RCIM_Cubic, tangentMode=RCTM_User, time=0, value=0}
         {interpMode=RCIM_Cubic, tangentMode=RCTM_User, time=0.1, value=1}
         {tangentMode=RCTM_User, time=1, value=0}
        ]}}})
    Scale RGB=&MakeVectorFromLinearColorRGB(Color=&FNTimeOfDayColor(
      Night Color=<rgba 0.3734 0.4516 0.5781 1>
      Morning Color=<rgba 0.5365 0.5365 0.5365 1>
      Day Color=<rgba 1 1 1 1>
      Evening Color=<rgba 0.7917 0.791 0.5113 1>
     ))
    ScaleA=true
    ScaleRGB=true
   )
  ]
  rendererClasses [&NiagaraMeshRendererProperties]
  renderers [{rendererClass=&NiagaraMeshRendererProperties, rendererIndex=0}]
 }
 OracleBodySplineAdditionalShape {
  emitterUpdateScript [
   &EmitterState(
    Inactive Response=::NewEnumerator1
    Life Cycle Mode=::NewEnumerator1
    Loop Behavior=::NewEnumerator1
    Loop Duration Mode=::NewEnumerator1
    UseLoopDelay=false
   )
   &SpawnBurst_Instantaneous(Use Spawn Probability=false, UseLoopCountLimit=false)
   &SpawnBurst_Instantaneous(Spawn Time=1.5135, Use Spawn Probability=false, UseLoopCountLimit=false) {module=SpawnBurst_Instantaneous001}
  ]
  particleSpawnScript [
   &InitializeParticle(
    Color Mode=::NewEnumerator0
    Lifetime=5
    Mesh Scale Mode=::NewEnumerator3
    Sprite Size Mode=::NewEnumerator3
    Uniform Sprite Size=150
    UsePositionOffset=true
    Write Material Random=true
   )
   &InitialMeshOrientation(Mesh Orientation Mode=::NewEnumerator7, Rotation=<0 0 0.25>, Use Rotation Vector=true)
  ]
  particleUpdateScript [
   &ParticleState(Kill Particles When Lifetime Has Elapsed=false)
   &ScaleColor(
    Scale RGB=&MakeVectorFromLinearColorRGB(Color=&MultiplyLinearColor(
      Linear Color=&FNTimeOfDayColor(
       Night Color=<rgba 0.3734 0.4516 0.5781 1>
       Morning Color=<rgba 0.5365 0.5365 0.5365 1>
       Day Color=<rgba 1 1 1 1>
       Evening Color=<rgba 0.7917 0.791 0.5113 1>
      )
      Scale Factor=&MakeVector4_FromLinearColor(LinearColor=&ColorFromCurve(
        ColorCurve=DataInterface() {propertyValues {
          AlphaCurve {keys [<at 0 1 RCIM_Cubic>]}
          BlueCurve {keys [<at 0 1 RCIM_Cubic>,<at 0.3818 4>,<at 0.6105 4>,<at 1 1 RCIM_Cubic>]}
          GreenCurve {keys [<at 0 1 RCIM_Cubic>,<at 0.3818 3.019>,<at 0.6105 3.0658>,<at 1 1 RCIM_Cubic>]}
          RedCurve {keys [<at 0 0.8694>,<at 0.3818 1.8125>,<at 0.6105 1.9167>,<at 1 1 RCIM_Cubic>]}
         }}
        CurveIndex=$System.Timeline.NormalizedTime
       ))
     ))
    ScaleA=true
    ScaleRGB=true
   )
   /SetVariables(Particles.GlowColor=&LerpLinearColors(StartColor=<rgba 0.2258 0.6928 1 0.5>, EndColor=<rgba 0.2258 0.6928 1 1>, LerpFactor=$System.DebugTalkAlpha)) {isSetParameters=true}
   &ScaleMeshSize(Scale Factor=&VectorFromFloat(Value=&Lerp_Float(
      A=&SelectFloatFromArray(Direct Array Index=$System.CurrentShapeIndex, Float Selection Array=$System.SecondShapeScaleArray)
      B=&SelectFloatFromArray(Direct Array Index=$System.NextShapeIndex, Float Selection Array=$System.SecondShapeScaleArray)
      Alpha=$System.Timeline.NormalizedTime
     ))) {module=ScaleMeshSize001}
   /SetVariables(Particles.MeshOrientation=&LerpQuaternion(
     Quaternion A:Quat4f=$Particles.Initial.MeshOrientation
     Quaternion B:Quat4f=&MultiplyQuaternion(
      Quaternion A=$Particles.Initial.MeshOrientation
      Quaternion B=&MakeQuatFromAxisAngle(Axis=<0 1 0>, AngleInDegrees=&Multiply_Float(A=$System.Age, B=360))
     )
     Lerp Factor:NiagaraFloat=&MakeCustomFloatFromBool(
      Bool=$System.IsTransitioning
      True Float=&Multiply_Float(
       A=$System.Timeline.NormalizedTime
       B=&SelectFloatFromArray(Direct Array Index=$System.NextShapeIndex, Float Selection Array=DataInterface() {propertyValues {FloatData [0,1,0]}})
      )
      False Float=&Multiply_Float(
       A=1
       B=&SelectFloatFromArray(Direct Array Index=$System.CurrentShapeIndex, Float Selection Array=DataInterface() {propertyValues {FloatData [0,1,0,0,0]}})
      )
     )
     Enable Spherical Linear Interpolation:NiagaraBool=false
    )) {isSetParameters=true, module=SetVariables_#17}
   /SetVariables(Particles.Position=&AddVectorToPosition(
     Position=$Engine.Emitter.SimulationPosition
     Vector=&Multiply_VectorByFloat(
      Vector=&ReturnCameraVector(Camera Vectors=::NewEnumerator2, Coordinate Space=::Local)
      Float=&Lerp_Float(
       A=&SelectFloatFromArray(Direct Array Index=$System.CurrentShapeIndex, Float Selection Array=DataInterface() {propertyValues {FloatData [-40,-40,-40,-40,-40]}})
       B=&SelectFloatFromArray(Direct Array Index=$System.NextShapeIndex, Float Selection Array=DataInterface() {propertyValues {FloatData [-40,-40,-40,-40,-40]}})
       Alpha=$System.Timeline.NormalizedTime
      )
     )
    )) {isSetParameters=true}
   &DynamicMaterialParameters(
    Index 0 Param 1=&Max_Float(A:NiagaraFloat=$System.DebugTalkAlpha, B:NiagaraFloat=0)
    Index 0 Param 2=&MakeFloatFromInt(Int=&SelectIntFromArray(Direct Array Index=$System.CurrentShapeIndex, Int Selection Array=DataInterface() {propertyValues {IntData [0,1,4,3,4]}}))
    Index 0 Param 3=&MakeFloatFromInt(Int=&SelectIntFromArray(Direct Array Index=$System.NextShapeIndex, Int Selection Array=DataInterface() {propertyValues {IntData [0,1,4,3,4]}}))
    Index 0 Param 4=$System.Timeline.NormalizedTime
    Index 1 Param 1=&Multiply_Float(
     A=&Multiply_Float(
      A=&Lerp_Float(
       A=&SelectFloatFromArray(Direct Array Index=$System.CurrentShapeIndex, Float Selection Array=$System.SecondShapeBaseWidthArray)
       B=&SelectFloatFromArray(Direct Array Index=$System.NextShapeIndex, Float Selection Array=$System.SecondShapeBaseWidthArray)
       Alpha=$System.Timeline.NormalizedTime
      )
      B=&FloatFromCurve(
       FloatCurve=DataInterface() {propertyValues {Curve {keys [
           {interpMode=RCIM_Cubic, tangentMode=RCTM_User, time=0, value=1}
           {interpMode=RCIM_Cubic, tangentMode=RCTM_User, time=0.5, value=0.3}
           {interpMode=RCIM_Cubic, tangentMode=RCTM_User, time=1, value=1}
          ]}}}
       CurveIndex=$System.Timeline.NormalizedTime
      )
     )
     B=$System.SpawnAlpha
    )
    Index 1 Param 2=&Multiply_Float(
     A=&Multiply_Float(
      A=&Lerp_Float(
       A=&SelectFloatFromArray(Direct Array Index=$System.CurrentShapeIndex, Float Selection Array=$System.SecondShapeTipWidthArray)
       B=&SelectFloatFromArray(Direct Array Index=$System.NextShapeIndex, Float Selection Array=$System.SecondShapeTipWidthArray)
       Alpha=$System.Timeline.NormalizedTime
      )
      B=&FloatFromCurve(
       FloatCurve=DataInterface() {propertyValues {Curve {keys [
           {interpMode=RCIM_Cubic, tangentMode=RCTM_User, time=0, value=1}
           {interpMode=RCIM_Cubic, tangentMode=RCTM_User, time=0.5, value=0.3}
           {interpMode=RCIM_Cubic, tangentMode=RCTM_User, time=1, value=1}
          ]}}}
       CurveIndex=$System.Timeline.NormalizedTime
      )
     )
     B=$System.SpawnAlpha
    )
    Index 1 Param 3=&Lerp_Float(
     A=0
     B=0.9
     Alpha=&FloatFromCurve(
      FloatCurve=DataInterface() {propertyValues {Curve {keys [
          {arriveTangent=2.1742, interpMode=RCIM_Cubic, leaveTangent=2.1742, tangentMode=RCTM_User, time=0, value=0}
          {arriveTangent=0.3068, interpMode=RCIM_Cubic, leaveTangent=0.3068, tangentMode=RCTM_User, time=1, value=1}
         ]}}}
      CurveIndex=$System.SpawnAlpha
     )
    )
    Index 1 Param 4=&Lerp_Float(
     A=&SelectFloatFromArray(Direct Array Index=$System.CurrentShapeIndex, Float Selection Array=$System.DarkenHeightArray)
     B=&SelectFloatFromArray(Direct Array Index=$System.NextShapeIndex, Float Selection Array=$System.DarkenHeightArray)
     Alpha=$System.Timeline.NormalizedTime
    )
    Index 2 Param 1=&Lerp_Float(
     A=&SelectFloatFromArray(Direct Array Index=$System.CurrentShapeIndex, Float Selection Array=DataInterface() {propertyValues {FloatData [1,0]}})
     B=&SelectFloatFromArray(Direct Array Index=$System.NextShapeIndex, Float Selection Array=DataInterface() {propertyValues {FloatData [1,0]}})
     Alpha=$System.Timeline.NormalizedTime
    )
    Module.Index0Param0Write=true
    Module.Index0Param1Write=true
    Module.Index0Param2Write=true
    Module.Index0Param3Write=true
    Module.Index1Param0Write=true
    Module.Index1Param1Write=true
    Module.Index1Param2Write=true
    Module.Index1Param3Write=true
    Module.Index2Param0Write=true
    Module.Index2Param1Write=true
    Module.Index2Param2Write=true
    Module.Index2Param3Write=true
    Write Parameter Index 1=true
    Write Parameter Index 2=true
   )
   /SetVariables(Particles.GlowDynMat1:Vector4f=&MakeVector4(
     X=&Multiply_Float(
      A=&MakeFloatFromVector4(Vector 4=$Particles.DynamicMaterialParameter1, Channel=::NewEnumerator0)
      B=1.7
     )
     Y=&Multiply_Float(
      A=&MakeFloatFromVector4(Vector 4=$Particles.DynamicMaterialParameter1, Channel=::NewEnumerator1)
      B=1.7
     )
     Z=&MakeFloatFromVector4(Vector 4=$Particles.DynamicMaterialParameter1, Channel=::NewEnumerator2)
     W=&MakeFloatFromVector4(Vector 4=$Particles.DynamicMaterialParameter1, Channel=::NewEnumerator3)
    )) {isSetParameters=true}
  ]
  rendererClasses [&NiagaraMeshRendererProperties]
  renderers [
   {rendererClass=&NiagaraMeshRendererProperties, rendererIndex=0}
   {rendererClass=&NiagaraMeshRendererProperties, rendererIndex=1}
  ]
 }
 OracleBodySplineBaseShape {
  emitterUpdateScript [
   &EmitterState(
    Inactive Response=::NewEnumerator1
    Life Cycle Mode=::NewEnumerator1
    Loop Behavior=::NewEnumerator1
    Loop Duration Mode=::NewEnumerator1
    UseLoopDelay=false
   )
   &SpawnBurst_Instantaneous(Use Spawn Probability=false, UseLoopCountLimit=false)
  ]
  particleSpawnScript [
   &InitializeParticle(
    Color Mode=::NewEnumerator0
    Lifetime=5
    Mesh Scale Mode=::NewEnumerator3
    Sprite Size Mode=::NewEnumerator3
    Uniform Sprite Size=150
    UsePositionOffset=true
    Write Material Random=true
   )
   &InitialMeshOrientation(Mesh Orientation Mode=::NewEnumerator7, Rotation=<0 0 -0.25>, Use Rotation Vector=true)
  ]
  particleUpdateScript [
   &ParticleState(Kill Particles When Lifetime Has Elapsed=false)
   &ScaleColor(
    Scale RGB=&MakeVectorFromLinearColorRGB(Color=&MultiplyLinearColor(
      Linear Color=&MultiplyLinearColor(
       Linear Color=&FNTimeOfDayColor(
        Night Color=<rgba 0.3734 0.4516 0.5781 1>
        Morning Color=<rgba 0.5365 0.5365 0.5365 1>
        Day Color=<rgba 1 1 1 1>
        Evening Color=<rgba 0.7917 0.791 0.5113 1>
       )
       Scale Factor=&MakeVector4_FromLinearColor(LinearColor=&ColorFromCurve(
         ColorCurve=DataInterface() {propertyValues {
           AlphaCurve {keys [<at 0 1 RCIM_Cubic>]}
           BlueCurve {keys [<at 0 1 RCIM_Cubic>,<at 0.3818 4>,<at 0.6105 4>,<at 1 1 RCIM_Cubic>]}
           GreenCurve {keys [<at 0 1 RCIM_Cubic>,<at 0.3818 3.019>,<at 0.6105 3.0658>,<at 1 1 RCIM_Cubic>]}
           RedCurve {keys [<at 0 0.8694>,<at 0.3818 1.8125>,<at 0.6105 1.9167>,<at 1 1 RCIM_Cubic>]}
          }}
         CurveIndex=$System.Timeline.NormalizedTime
        ))
      )
      Scale Factor=<1 1 1 1>
     ))
    ScaleA=true
    ScaleRGB=true
   )
   /SetVariables(Particles.GlowColor=&LerpLinearColors(StartColor=<rgba 0.2258 0.6928 1 0.5>, EndColor=<rgba 0.2258 0.6928 1 1>, LerpFactor=$System.DebugTalkAlpha)) {isSetParameters=true}
   /SetVariables(Particles.MeshOrientation=&LerpQuaternion(
     Quaternion A:Quat4f=$Particles.Initial.MeshOrientation
     Quaternion B:Quat4f=&MultiplyQuaternion(
      Quaternion A=$Particles.Initial.MeshOrientation
      Quaternion B=&MakeQuatFromAxisAngle(Axis=<0 1 0>, AngleInDegrees=&Multiply_Float(A=$System.Age, B=360))
     )
     Lerp Factor:NiagaraFloat=&MakeCustomFloatFromBool(
      Bool=$System.IsTransitioning
      True Float=&Multiply_Float(
       A=&FloatFromCurve(FloatCurve=DataInterface() {propertyValues {Curve {keys [<at 0 0>,<at 0.8 1>]}}}, CurveIndex=$System.Timeline.NormalizedTime)
       B=&SelectFloatFromArray(Direct Array Index=$System.NextShapeIndex, Float Selection Array=DataInterface() {propertyValues {FloatData [0,1,0]}})
      )
      False Float=&Multiply_Float(
       A=1
       B=&SelectFloatFromArray(Direct Array Index=$System.CurrentShapeIndex, Float Selection Array=DataInterface() {propertyValues {FloatData [0,1,0]}})
      )
     )
     Enable Spherical Linear Interpolation:NiagaraBool=false
    )) {isSetParameters=true, module=SetVariables_#17}
   &DynamicMaterialParameters(
    Index 0 Param 1=&Max_Float(
     A:NiagaraFloat=$System.DebugTalkAlpha
     B:NiagaraFloat=&Multiply_Float(A=$System.Timeline002.NormalizedTime, B=&Lerp_Float(A=0.5, B=1, Alpha=$System.GlowIntensity))
    )
    Index 0 Param 2=&MakeFloatFromInt(Int=$System.CurrentShapeIndex)
    Index 0 Param 3=&MakeFloatFromInt(Int=$System.NextShapeIndex)
    Index 0 Param 4=$System.Timeline.NormalizedTime
    Index 1 Param 1=&Multiply_Float(
     A=&Multiply_Float(
      A=&Lerp_Float(
       A=&SelectFloatFromArray(Direct Array Index=$System.CurrentShapeIndex, Float Selection Array=$System.BaseWidthArray)
       B=&SelectFloatFromArray(Direct Array Index=$System.NextShapeIndex, Float Selection Array=$System.BaseWidthArray)
       Alpha=$System.Timeline.NormalizedTime
      )
      B=&FloatFromCurve(
       FloatCurve=DataInterface() {propertyValues {Curve {keys [
           {interpMode=RCIM_Cubic, tangentMode=RCTM_User, time=0, value=1}
           {interpMode=RCIM_Cubic, tangentMode=RCTM_User, time=0.5, value=0.3}
           {interpMode=RCIM_Cubic, tangentMode=RCTM_User, time=1, value=1}
          ]}}}
       CurveIndex=$System.Timeline.NormalizedTime
      )
     )
     B=1
    )
    Index 1 Param 2=&Multiply_Float(
     A=&Multiply_Float(
      A=&Lerp_Float(
       A=&SelectFloatFromArray(Direct Array Index=$System.CurrentShapeIndex, Float Selection Array=$System.TipWidthArray)
       B=&SelectFloatFromArray(Direct Array Index=$System.NextShapeIndex, Float Selection Array=$System.TipWidthArray)
       Alpha=$System.Timeline.NormalizedTime
      )
      B=&FloatFromCurve(
       FloatCurve=DataInterface() {propertyValues {Curve {keys [
           {interpMode=RCIM_Cubic, tangentMode=RCTM_User, time=0, value=1}
           {interpMode=RCIM_Cubic, tangentMode=RCTM_User, time=0.5, value=0.3}
           {interpMode=RCIM_Cubic, tangentMode=RCTM_User, time=1, value=1}
          ]}}}
       CurveIndex=$System.Timeline.NormalizedTime
      )
     )
     B=1
    )
    Index 1 Param 3=&Lerp_Float(
     A=0
     B=0.9
     Alpha=&FloatFromCurve(
      FloatCurve=DataInterface() {propertyValues {Curve {keys [
          {arriveTangent=3.4736, interpMode=RCIM_Cubic, leaveTangent=3.4736, tangentMode=RCTM_User, time=0, value=0}
          {arriveTangent=0.3068, interpMode=RCIM_Cubic, leaveTangent=0.3068, tangentMode=RCTM_User, time=1, value=1}
         ]}}}
      CurveIndex=$System.SpawnAlpha
     )
    )
    Index 1 Param 4=&Lerp_Float(
     A=&SelectFloatFromArray(Direct Array Index=$System.CurrentShapeIndex, Float Selection Array=$System.DarkenHeightArray)
     B=&SelectFloatFromArray(Direct Array Index=$System.NextShapeIndex, Float Selection Array=$System.DarkenHeightArray)
     Alpha=$System.Timeline.NormalizedTime
    )
    Index 2 Param 1=&Lerp_Float(
     A=&SelectFloatFromArray(Direct Array Index=$System.CurrentShapeIndex, Float Selection Array=DataInterface() {propertyValues {FloatData [1,0]}})
     B=&SelectFloatFromArray(Direct Array Index=$System.NextShapeIndex, Float Selection Array=DataInterface() {propertyValues {FloatData [1,0]}})
     Alpha=$System.Timeline.NormalizedTime
    )
    Module.Index0Param0Write=true
    Module.Index0Param1Write=true
    Module.Index0Param2Write=true
    Module.Index0Param3Write=true
    Module.Index1Param0Write=true
    Module.Index1Param1Write=true
    Module.Index1Param2Write=true
    Module.Index1Param3Write=true
    Module.Index2Param0Write=true
    Module.Index2Param1Write=true
    Module.Index2Param2Write=true
    Module.Index2Param3Write=true
    Write Parameter Index 1=true
    Write Parameter Index 2=true
   )
   /SetVariables(Particles.GlowDynMat1:Vector4f=&MakeVector4(
     X=&Multiply_Float(
      A=&Multiply_Float(
       A=&Lerp_Float(
        A=&SelectFloatFromArray(Direct Array Index=$System.CurrentShapeIndex, Float Selection Array=$System.GlowBaseWidthArray)
        B=&SelectFloatFromArray(Direct Array Index=$System.NextShapeIndex, Float Selection Array=$System.GlowBaseWidthArray)
        Alpha=$System.Timeline.NormalizedTime
       )
       B=&FloatFromCurve(
        FloatCurve=DataInterface() {propertyValues {Curve {keys [
            {interpMode=RCIM_Cubic, tangentMode=RCTM_User, time=0, value=1}
            {interpMode=RCIM_Cubic, tangentMode=RCTM_User, time=0.5, value=0.3}
            {interpMode=RCIM_Cubic, tangentMode=RCTM_User, time=1, value=1}
           ]}}}
        CurveIndex=$System.Timeline.NormalizedTime
       )
      )
      B=$System.SpawnAlpha
     )
     Y=&Multiply_Float(
      A=&Multiply_Float(
       A=&Lerp_Float(
        A=&SelectFloatFromArray(Direct Array Index=$System.CurrentShapeIndex, Float Selection Array=$System.GlowTipWidthArray)
        B=&SelectFloatFromArray(Direct Array Index=$System.NextShapeIndex, Float Selection Array=$System.GlowTipWidthArray)
        Alpha=$System.Timeline.NormalizedTime
       )
       B=&FloatFromCurve(
        FloatCurve=DataInterface() {propertyValues {Curve {keys [
            {interpMode=RCIM_Cubic, tangentMode=RCTM_User, time=0, value=1}
            {interpMode=RCIM_Cubic, tangentMode=RCTM_User, time=0.5, value=0.3}
            {interpMode=RCIM_Cubic, tangentMode=RCTM_User, time=1, value=1}
           ]}}}
        CurveIndex=$System.Timeline.NormalizedTime
       )
      )
      B=$System.SpawnAlpha
     )
     Z=&MakeFloatFromVector4(Vector 4=$Particles.DynamicMaterialParameter1, Channel=::NewEnumerator2)
     W=&MakeFloatFromVector4(Vector 4=$Particles.DynamicMaterialParameter1, Channel=::NewEnumerator3)
    )) {isSetParameters=true}
  ]
  rendererClasses [&NiagaraMeshRendererProperties]
  renderers [
   {rendererClass=&NiagaraMeshRendererProperties, rendererIndex=0}
   {rendererClass=&NiagaraMeshRendererProperties, rendererIndex=1}
  ]
 }
}
summary {
 systemName=NS_S29_OracleBody_Idle
 emitters [
  {emitterName=Glow, rendererClasses [&NiagaraSpriteRendererProperties]}
  {emitterName=NE_StarGlow, rendererClasses [&NiagaraSpriteRendererProperties]}
  {emitterName=Dropping, simTarget=GPUComputeSim, rendererClasses [&NiagaraSpriteRendererProperties]}
  {emitterName=ConstantSplashes, simTarget=GPUComputeSim, rendererClasses [&NiagaraSpriteRendererProperties]}
  {emitterName=OracleBodySplineBaseShape, rendererClasses [&NiagaraMeshRendererProperties]}
  {emitterName=OracleBodySplineAdditionalShape, rendererClasses [&NiagaraMeshRendererProperties]}
  {emitterName=NE_WaterRipples, rendererClasses [&NiagaraMeshRendererProperties]}
  {emitterName=DropletsImpact, simTarget=GPUComputeSim, rendererClasses [&NiagaraSpriteRendererProperties]}
 ]
 userVariables [
  User.BaseShapeIndex {type=NiagaraInt32}
  User.IsGlowing {type=NiagaraBool, defaultValue=true}
  User.IsSpawning {type=NiagaraBool, defaultValue=true}
  User.IsTalking {type=NiagaraBool}
  User.IsTransitioning {type=NiagaraBool}
  User.NextShapeIndex {type=NiagaraInt32}
  User.StateSwitchSpeed {type=NiagaraFloat, defaultValue=2.2, description=""}
  User.SystemScale {type=NiagaraFloat, defaultValue=1, description=""}
  User.TalkAlpha {type=NiagaraFloat, defaultValue=0, description=""}
 ]
}
